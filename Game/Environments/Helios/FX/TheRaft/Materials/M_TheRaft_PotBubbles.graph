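# Masked, lit material for Niagara mesh particles that renders pot bubbles in The Raft environment effects.
# A panning wisp texture plus a UV gradient drives particle-controlled erosion and distorts randomly offset cell noise, which blends dark and light bubble colors.
# View-angle falloff blends emissive tints, and a parameter switches to particle color.
# Cell noise, sampled by panning camera-space-normal UVs with per-particle randomness, scales displacement along vertex normals.
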Material /Game/Environments/Helios/FX/TheRaft/Materials/M_TheRaft_PotBubbles
graph {
 properties {bAutomaticallySetUsageInEditor=false, bUsedWithNiagaraMeshParticles=true, blendMode=BLEND_Masked, usedShadingModels=Default Lit}
 nodes [
  n1 = TextureSampleParameter2D(UVs=Panner() {speedX=0.1, speedY=0.2}) {parameterName=Dissolve, texture=/Game/Environments/Helios/FX/TheRaft/T_TheRaft_Wisp}
  n2 = ComponentMask(None=TextureCoordinate()) {g=true}
  n3 = DynamicParameter() {defaultValue=<rgba 0 1 1 1.4>, paramNames [Erosion, RandomU, RandomV, DistortionAmount]}
  Add(A=Multiply(A=@n1.R, B=@n2), B=LinearInterpolate(Alpha=@n3.Erosion) {constA=1, constB=-1})
  Constant() {r=0.1}
  n4 = TextureSample() {texture=/Game/Global/Textures/Default/Noise/3xPerlin_M}
  n5 = ScalarParameter() {defaultValue=0.2, parameterName=Noise Mult}
  n6 = TextureCoordinate()
  n7 = DynamicParameter() {defaultValue=<rgba 0 1 1 1.4>, paramNames [Erosion, RandomU, RandomV, DistortionAmount]}
  n8 = AppendVector(A=@n7.RandomU, B=@n7.RandomU)
  n9 = TextureSampleParameter2D(UVs=Add(A=Add(A=@n6, B=@n8), B=Multiply(A=@n1.R) {constB=0.2})) {parameterName=MainMask, texture=/Game/Environments/Helios/FX/TheRaft/T_TheRaft_CellNoise}
  n10 = ComponentMask(None=TextureCoordinate()) {g=true}
  n11 = Multiply(A=Saturate(None=Multiply(A=@n9.R) {constB=5}), B=SmoothStep(Value=OneMinus(None=@n10)) {constMax=0.7, constMin=0.45})
  n12 = LinearInterpolate(
   A=VectorParameter() {parameterName=BubbleDarkColor, defaultValue=<rgba 0.276 0.1497 0.0737 1>}.RGB
   B=VectorParameter() {parameterName=BubbleLighterColor, defaultValue=<rgba 0.6172 0.402 0.117 1>}.RGB
   Alpha=@n11
  )
  n13 = ParticleColor()
  n14 = LinearInterpolate(A=Multiply(A=@n13.RGB) {constB=0.276}, B=@n13.RGB, Alpha=@n11)
  n15 = ScalarParameter() {parameterName=Use_ParticleColor}
  Desaturation(None=Add(A=Multiply(A=@n4.RGB, B=@n5) {constB=0.2}, B=LinearInterpolate(A=@n12, B=@n14, Alpha=@n15)), Fraction=Constant() {r=0.1})
  n16 = ParticleColor()
  n17 = DotProduct(A=VertexNormalWS(), B=CameraVectorWS())
  n18 = Saturate(None=Multiply(A=Power(Base=@n17) {constExponent=4}) {constB=5})
  n19 = VectorParameter() {parameterName=EmissiveColor, defaultValue=<rgba 0.0345 0.2552 0 1>}
  n20 = VectorParameter() {parameterName=EmissiveColor2, defaultValue=<rgba 0.1353 1 0 1>}
  LinearInterpolate(
   A=Multiply(A=LinearInterpolate(A=@n19.RGB, B=@n20.RGB, Alpha=@n18), B=ScalarParameter() {parameterName=EmissiveIntensity})
   B=LinearInterpolate(A=Multiply(A=@n16.RGB) {constB=0.2}, B=@n16.RGB, Alpha=@n18)
   Alpha=ScalarParameter() {parameterName=Use_ParticleColor}
  )
  Power(Base=@n2)
  n21 = VertexNormalWS()
  n22 = Transform(None=VertexNormalWS()) {transformSourceType=TRANSFORMSOURCE_World, transformType=TRANSFORM_Camera}
  n23 = Add(A=ComponentMask(None=Multiply(A=@n22) {constB=0.5}) {g=true, r=true}) {constB=0.5}
  n24 = ParticleRandom()
  n25 = Panner(Coordinate=Multiply(A=Add(A=@n23, B=@n24)) {constB=0.2}) {speedX=0.1, speedY=0.2}
  Reroute(None=Multiply(
    A=Multiply(A=@n21, B=@n3.DistortionAmount) {constB=1.4}
    B=TextureSampleParameter2D(UVs=@n25) {parameterName=WPOMask, texture=&T_TheRaft_CellNoise}.R
   ))
  TextureCoordinate()
 ]
}
