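# Niagara system for the waterfall effect in the Helios TheRaft environment.
# It holds four emitters named for their layers: VerticalDrops, BaseDrops, Splash and SplashFlat.
# It exposes a User.OffsetPosition position parameter, presumably to shift the effect.
# SystemState is the only module listed, and no renderers appear for any emitter, so the emitters' internal behavior is not visible here.
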NiagaraSystem /Game/Environments/Helios/FX/TheRaft/NSLW_TheRaft_Waterfall
dependencies {usedDataInterfaces [], usedDynamicInputs [], usedModules [&SystemState]}
emitters {
 BaseDrops {rendererClasses [], renderers []}
 Splash {rendererClasses [], renderers []}
 SplashFlat {rendererClasses [], renderers []}
 VerticalDrops {rendererClasses [], renderers []}
}
summary {
 systemName=NSLW_TheRaft_Waterfall
 emitters [
  {emitterName=SplashFlat, rendererClasses []}
  {emitterName=Splash, rendererClasses []}
  {emitterName=BaseDrops, rendererClasses []}
  {emitterName=VerticalDrops, rendererClasses []}
 ]
 userVariables [User.OffsetPosition {type=NiagaraPosition, description="", defaultValue {}}]
}
