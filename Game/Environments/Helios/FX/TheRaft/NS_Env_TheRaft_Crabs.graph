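# Niagara environment effect for TheRaft: scattered crabs that scuttle over the landscape and burrow into sand based on player distance.
# A distance-culled GPU emitter spawns a short burst scaled by user rate and radius, snapped to sampled landscape, with a state machine pulsing movement.
# A latched Burrow flag plays a timeline that sinks each crab, fades sand tint in and out by time of day, then kills it.
# Two separate sand emitters are disabled.
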NiagaraSystem /Game/Environments/Helios/FX/TheRaft/NS_Env_TheRaft_Crabs
dependencies {
 usedDataInterfaces [
  &NiagaraDataInterfaceCurve
  &NiagaraDataInterfaceParticleRead
  &NiagaraDataInterfaceMeshRendererInfo
  &NiagaraDataInterfaceSpriteRendererInfo
 ]
 usedDynamicInputs []
 usedModules [&SystemState]
}
emitters {
 Crabs {
  simTarget=GPUComputeSim
  emitterUpdateScript [
   &EmitterState(
    Enable Distance Culling=true
    Life Cycle Mode=::NewEnumerator1
    Loop Behavior=::NewEnumerator1
    Loop Duration Mode=::NewEnumerator1
    Max Distance Response=::SleepAndClearParticles
    MaxDistance=4500
    Min Distance Response=::SleepAndLetParticlesFinish
    Module.bApplySpawnCountScaleByDistance=true
    Module.bUseMaxDistance=true
    Module.bUseMinDistance=true
    Reset Age On Awaken=true
    Scalability Mode=::NewEnumerator1
    Scale Spawn Count=true
    Spawn Count Scale By Distance=DataInterface() {propertyValues {Curve {keys [<at 0 1>,<at 0.95 1>,<at 1 0>]}}}
    UseLoopDelay=false
   )
   &SpawnRate(
    SpawnRate=&FloatFromCurve(
     FloatCurve=DataInterface() {propertyValues {Curve {keys [<at 0 0>,<at 0.1 1>,<at 0.2 0>]}}}
     CurveIndex=$Emitter.Age
     Scale Curve=&Multiply_Float(A=200, B=$User.SpawnRate)
    )
    Use Spawn Probability=false
   )
  ]
  particleSpawnScript [
   &InitializeParticle(
    AdjustAlpha=false
    AdjustHue=true
    AdjustSaturation=true
    AdjustValue=true
    Color Mode=::NewEnumerator3
    Lifetime=0.6
    Mesh Scale Mode=::NewEnumerator1
    Mesh Uniform Scale Max=3
    Mesh Uniform Scale Min=1.5
    Saturation Range:Vector2f=<0.8 500>
    UsePositionOffset=false
    Value Range:Vector2f=<2.5 6>
    Write Material Random=true
   )
   &self:Sample_Landscape() {module=Sample_Landscape}
   &ShapeLocation(
    Cylinder Height=1
    Cylinder Radius=&Multiply_Float(A=600, B=$User.RadiusMult)
    MaskSpawnGroup=false
    Offset=&RandomRangeVector(Minimum=<-99 -99 0>, Maximum=<300 300 0>)
    Offset Mode=::NewEnumerator0
    Override Seed=false
    Shape Primitive=::NewEnumerator1
    Surface Only=false
    UseOwnerScale=true
   )
   /SetVariables(Particles.Movement:Vector3f=&RandomRangeVector(Minimum=<-1 -0.25 0>, Maximum=<1 0.25 0>)) {isSetParameters=true}
   &InitialMeshOrientation(Mesh Axis To Orient=<-1 0 0>, Mesh Orientation Mode=::NewEnumerator2, Orientation Vector=$Particles.Movement, Use Rotation Vector=false)
   /SetVariables() {isSetParameters=true}
  ]
  particleUpdateScript [
   &ParticleState(Kill Particles When Lifetime Has Elapsed=false, Let Infinitely Lived Particles Die When Emitter Deactivates=true) {module=ParticleState001}
   /Niagara/Modules/Update/StateChange/TimeBasedStateMachine(
    Lerp Time To ON:NiagaraFloat=&RandomRangeFloat(Minimum=0.5, Maximum=1)
    Lerp Time To Off:NiagaraFloat=&RandomRangeFloat(Minimum=0.5, Maximum=1)
    Off Duration:NiagaraFloat=&RandomRangeFloat(Minimum=2, Maximum=3)
    On Duration:NiagaraFloat=&RandomRangeFloat(Minimum=0.4, Maximum=0.6)
   )
   /SetVariables(Particles.Burrow:NiagaraBool=&Boolean_OrOperation(
     Bool1=$Particles.Burrow
     Bool2=&SetBoolByFloatComparison(
      A=500
      B=&DistanceBetweenPositions(A=$Particles.Position, B=&MakeVectorFromLinearColorRGB(Color=$NPC.FortniteNPC.PlayerPosition))
      Comparison Type=::NewEnumerator2
     )
    )) {isSetParameters=true}
   /SetVariables(Particles.BurrowDust:NiagaraInt32=&MakeCustomIntFromBool(Bool=$Particles.Burrow, True INT=1, False INT=0)) {isSetParameters=true}
   &Timeline(Play=$Particles.Burrow)
   &KillParticles(Kill Particles=&SetBoolByFloatComparison(A=1, B=$Particles.Timeline.Time, Comparison Type=::NewEnumerator4))
   /SetVariables(Particles.Position=&AddVectorToPosition(
     Position=$Particles.Position
     Vector=&Multiply_VectorByFloat(Vector=$Particles.Movement, Float=&Multiply_Float(A=$Output.TimeBasedStateMachine.On Off Percentage, B=15))
    )) {isSetParameters=true}
   &self:Sample_Landscape() {module=Sample_Landscape001}
   /SetVariables(Particles.Position=&AddVectorToPosition(
     Position=$StackContext.Position
     Vector=&Multiply_VectorByFloat(Vector=<0 0 -10>, Float=$Particles.Timeline.NormalizedTime)
    )) {isSetParameters=true}
   &DynamicMaterialParameters(
    Index 0 Param 1=&RandomRangeFloat(Minimum=1, Maximum=2)
    Index 0 Param 2=&RandomRangeFloat(Minimum=0.5, Maximum=0.7)
    Index 0 Param 3=4
    Index 0 Param 4=&RandomRangeFloat(Minimum=0, Maximum=0.05)
    Index 1 Param 1=&RandomRangeFloat(Minimum=0.02, Maximum=0.05)
    Module.Index0Param0Write=true
    Module.Index0Param1Write=true
    Module.Index0Param2Write=true
    Module.Index0Param3Write=true
    Module.Index1Param0Write=true
    Module.Index1Param1Write=true
    Module.Index1Param2Write=true
    Module.Index1Param3Write=true
    Write Parameter Index 1=true
   )
   &SubUVAnimation(
    Random Start Frame=true
    "Sprite Renderer "=DataInterface() {propertyValues {SpriteRenderer=&self:Crabs_1.NiagaraSpriteRendererProperties_1}}
    UseEndFrame=false
    UseStartFrame=false
   )
   /SetVariables(Particles.SandCamera=55) {isSetParameters=true}
   /SetVariables(Particles.SandColor:LinearColor=&MultiplyLinearColor(
     Linear Color=<rgba 0.4078 0.2157 0.1176 1>
     Scale Factor=&MakeFloat4FromVectorAndFloat(
      "Vector (RGB)"=<1 1 1>
      "Float (Alpha)"=&FloatFromCurve(
       FloatCurve=DataInterface() {propertyValues {Curve {keys [
           {interpMode=RCIM_Cubic, tangentMode=RCTM_User, time=0, value=0}
           {interpMode=RCIM_Cubic, tangentMode=RCTM_User, time=0.5, value=1}
           {tangentMode=RCTM_User, time=1, value=0}
          ]}}}
       CurveIndex=$Particles.Timeline.NormalizedTime
       Scale Curve=&FN_Select_Float_from_Time_Of_Day(Morning=0.8, Day=1, Evening=0.8, Night=0.5)
      )
     )
    )) {isSetParameters=true}
  ]
  rendererClasses [&NiagaraMeshRendererProperties, &NiagaraSpriteRendererProperties]
  renderers [
   {rendererClass=&NiagaraMeshRendererProperties, rendererIndex=0}
   {rendererClass=&NiagaraSpriteRendererProperties, rendererIndex=1}
  ]
 }
 Sand {
  bEnabled=false
  simTarget=GPUComputeSim
  emitterUpdateScript [
   &EmitterState(Life Cycle Mode=::NewEnumerator1, Loop Duration=1, UseLoopDelay=false)
   &SpawnParticlesFromOtherEmitter(
    Attribute Reader=DataInterface() {propertyValues {EmitterBinding {bindingMode=Other, emitterName=Crabs001}}}
    Use Maximum Age=false
    Use Minimum Age=false
   )
  ]
  particleSpawnScript [
   &InitializeParticle(Color=<rgba 0 1 0 1>, Sprite Size Mode=::NewEnumerator3, Uniform Sprite Size=55, UsePositionOffset=false, Write Material Random=true)
   &SampleParticlesFromOtherEmitter()
  ]
  particleUpdateScript [
   &ParticleState()
   &SolveForcesAndVelocity(Clamp Velocity=false, Limit Acceleration=false)
  ]
  rendererClasses [&NiagaraSpriteRendererProperties]
  renderers [{rendererClass=&NiagaraSpriteRendererProperties}]
 }
 Sand001 {
  bEnabled=false
  simTarget=GPUComputeSim
  emitterUpdateScript [
   &EmitterState(Life Cycle Mode=::NewEnumerator1, Loop Duration=1, UseLoopDelay=false)
   &SpawnParticlesFromOtherEmitter(
    Attribute Reader=DataInterface() {propertyValues {EmitterBinding {bindingMode=Other, emitterName=Crabs001}}}
    Use Maximum Age=false
    Use Minimum Age=false
   )
  ]
  particleSpawnScript [
   &InitializeParticle(
    Color=<rgba 0 1 0 1>
    Sprite Rotation Angle=&RandomRangeFloat(Minimum=0, Maximum=360)
    Sprite Rotation Mode=::NewEnumerator0
    Sprite Size Mode=::NewEnumerator3
    Uniform Sprite Size=55
    UsePositionOffset=false
    Write Material Random=true
   )
   &SampleParticlesFromOtherEmitter()
   &SphereLocation(
    MaskSpawnGroup=false
    Override Seed=false
    Sphere Radius=5
    Surface Only=false
    UseNonUniformScale=false
    UseOffset=false
    UseSphereOrientation=false
   )
   &AddVelocity(Coordinate Space=::World, Velocity=&RandomRangeVector(Minimum=<0 0 0>, Maximum=<0 5 10>))
  ]
  particleUpdateScript [
   &ParticleState()
   &Drag(Drag=2, Use Linear Drag=true, Use Rotational Drag=false)
   &GravityForce(Gravity=&RandomRangeVector(Minimum=<0 0 -50>, Maximum=<1 1 -2>))
   &SpriteRotationRate(Rotation Rate=&RandomRangeFloat(Minimum=-30, Maximum=30))
   &WindForce(Wind Speed=&RandomRangeVector(Minimum=<20 -20 0>, Maximum=<60 60 1>))
   &AccelerationForce(Acceleration=<60 30 0>)
   &SolveForcesAndVelocity(Clamp Velocity=false, Limit Acceleration=false)
   &ScaleSpriteSize(Scale Factor=&Vector2DFromFloat(Value=&FloatFromCurve(
      FloatCurve=DataInterface() {propertyValues {Curve {keys [{interpMode=RCIM_Cubic, leaveTangent=1.5484, tangentMode=RCTM_Break, time=0, value=0.8}, <at 1 1.45>]}}}
      Scale Curve=&RandomRangeFloat(Minimum=0.8, Maximum=1.2354)
     )))
   &Color(
    Color=<rgba 0.4063 0.2152 0.1189 1>
    Scale Alpha=&FloatFromCurve(
     FloatCurve=DataInterface() {propertyValues {Curve {keys [
         {interpMode=RCIM_Cubic, leaveTangent=0.1136, tangentMode=RCTM_Break, time=0, value=0}
         {interpMode=RCIM_Cubic, leaveTangent=-1.8827, tangentMode=RCTM_Break, time=0.25, value=1}
         {arriveTangent=-0.027, interpMode=RCIM_Cubic, leaveTangent=-0.027, tangentMode=RCTM_User, time=1, value=0}
        ]}}}
     CurveIndex=$Emitter.NormalizedLoopAge
     Scale Curve=&MakeCustomFloatFromBool(
      Bool=HlslExpression() {hlslExpression="Emitter.LoopCount == 0"}
      True Float=&RandomRangeFloat(Minimum=0.5, Maximum=1)
      False Float=0
     )
    )
   ) {module=Color001}
   &DynamicMaterialParameters(
    Index 0 Param 1=&FloatFromCurve(FloatCurve=DataInterface() {propertyValues {Curve {keys [{interpMode=RCIM_Cubic, leaveTangent=-3.5204, tangentMode=RCTM_Break, time=0, value=1}, <at 1 -0.5>]}}})
    Module.Param0WriteEnabled=true
    Module.Param1WriteEnabled=true
    Module.Param2WriteEnabled=true
    Module.Param3WriteEnabled=true
   )
  ]
  rendererClasses [&NiagaraSpriteRendererProperties]
  renderers [
   {rendererClass=&NiagaraSpriteRendererProperties}
   {rendererClass=&NiagaraSpriteRendererProperties, rendererIndex=1}
  ]
 }
}
summary {
 systemName=NS_Env_TheRaft_Crabs
 emitters [
  {bEnabled=false, emitterName=Sand, simTarget=GPUComputeSim, rendererClasses [&NiagaraSpriteRendererProperties]}
  {bEnabled=false, emitterName=Sand001, simTarget=GPUComputeSim, rendererClasses [&NiagaraSpriteRendererProperties]}
  {
   emitterName=Crabs
   simTarget=GPUComputeSim
   rendererClasses [&NiagaraMeshRendererProperties, &NiagaraSpriteRendererProperties]
  }
 ]
 userVariables [
  User.Landscape {
   type=NiagaraDataInterfaceLandscape
   description=""
   defaultValue {
    struct=NiagaraExt_VariableValue_DataInterface
    value {
     dataInterface=&self:NiagaraDataInterfaceLandscape_0
     dataInterfaceClass=&NiagaraDataInterfaceLandscape
    }
   }
  }
  User.RadiusMult {type=NiagaraFloat, defaultValue=1, description=""}
  User.SpawnRate {type=NiagaraFloat, defaultValue=1, description=""}
 ]
}
