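# Niagara system for dripping water in the TheRaft sewer environment, using one GPU sprite emitter with distance culling and user-tunable spawn rate, fall speed and color.
# Spawn points from a wide plane snap to the nearest distance field surface, then to the depth buffer, with distance-based kills, so drips originate on visible geometry.
# Short-lived drops fall under curve-driven gravity, stretched by speed and drawn by three sprite renderers.
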NiagaraSystem /Game/Environments/Helios/FX/TheRaft/NS_Env_TheRaft_SewerDrips
dependencies {
 usedDataInterfaces [
  &NiagaraDataInterfaceCollisionQuery
  &NiagaraDataInterfaceCurve
  &NiagaraDataInterfaceVectorCurve
 ]
 usedDynamicInputs [&Lerp_Float]
 usedModules [
  &SystemState
  &self:SystemSpawnScript.SystemScriptSource.SystemScriptGraph.NiagaraNodeAssignment_1.SetVariables_#1
 ]
}
emitters {RainFromObjects {
  simTarget=GPUComputeSim
  emitterUpdateScript [
   &EmitterState(
    Enable Distance Culling=true
    Max Distance Response=::SleepAndClearParticles
    MaxDistance=15000
    Min Distance Response=::SleepAndLetParticlesFinish
    Module.bUseMaxDistance=true
    Module.bUseMinDistance=true
    Scalability Mode=::NewEnumerator1
   )
   &SpawnRate(SpawnRate=&Multiply_Float(A=35, B=$User.Rain_SpawnRateScale), Use Spawn Probability=false)
  ]
  particleSpawnScript [
   &InitializeParticle(
    Color=$User.Rain_NonLitColor
    Lifetime Max=1
    Lifetime Min=0.5
    Lifetime Mode=::NewEnumerator1
    Sprite Rotation Mode=::NewEnumerator1
    Sprite Size Max=<25 20>
    Sprite Size Min=<25 15>
    Sprite Size Mode=::NewEnumerator2
    UsePositionOffset=false
    Write Material Random=true
   ) {module=InitializeParticle001}
   &AddVelocity(Velocity=&Multiply_VectorByFloat(Vector=&RandomRangeVector(Minimum=<0 0 -50>, Maximum=<0 0 -50>), Float=$User.Rain_FallingSpeed))
   &ShapeLocation(
    Cylinder Height=$User.Rain_CylinderHeight
    Cylinder Radius=35
    MaskSpawnGroup=false
    Non Uniform Scale=<1 0.5 1>
    Override Seed=false
    Rotation Angle=90
    Rotation Angle Type:ENiagara_AngleInput=::NewEnumerator0
    Rotation Axis=<0 90 0>
    Rotation Mode=::NewEnumerator1
    Shape Primitive=::NewEnumerator1
    Surface Only=false
    UseOwnerScale=true
   ) {enabled=false}
   &ShapeLocation(
    Box / Plane Mode:ENiagara_BoxPlaneMode=::NewEnumerator1
    MaskSpawnGroup=false
    Override Seed=false
    Plane Size:Vector2f=<2500 4250>
    Shape Primitive=::NewEnumerator2
    UseOwnerScale=true
   ) {module=ShapeLocation001}
   /Niagara/Modules/Math/MoveToNearestDistanceFieldSurface_GPU(Kill by Distance=1200, bKillByDistance=true)
   /SetVariables(Particles.InitialPos:NiagaraPosition=$Particles.Position) {isSetParameters=true}
   /Niagara/Modules/Math/PlaceParticlesOnDepthBuffer_GPU(Max Z Depth=200, Module.Use Max Z Depth=true)
   &KillParticles(Kill Particles=&SetBoolByFloatComparison(
     A=&VectorLength(Vector=&Subtract_Vector(A=$Particles.InitialPos, B=Linked() {linkedVariable {name=Particles.Position, type=NiagaraPosition}}))
     B=1200
    ))
   &KillParticles(Kill Particles=&Boolean_NotOperation(Bool=$Output.PlaceParticlesOnDepthBuffer_GPU.ValidSample)) {enabled=false, module=KillParticles001}
   &SubUVAnimation(End Frame=3, SubUV Animation Mode:ENiagara_SubUVLookupMode=::NewEnumerator2)
   &CameraOffset(Camera Offset Amount=-10)
   &self:ScratchModule_02(Angle=90, BlendDirection=1, Normal:Vector3f=<0 0 -1>, PickCeilingOrFloor=true, Velocity=-200) {enabled=false, module=ScratchModule_02}
  ]
  particleUpdateScript [
   &ParticleState()
   &ScaleSpriteSize(Uniform Curve Sprite Scale=DataInterface() {propertyValues {Curve {keys [{tangentMode=RCTM_User, time=0, value=0.85}, {tangentMode=RCTM_User, time=1, value=1}]}}})
   &GravityForce(Gravity=&VectorFromCurve(
     VectorCurve=DataInterface() {propertyValues {
       XCurve {keys [<at 0 1>,<at 1 0>]}
       YCurve {keys [<at 0 1>,<at 1 0>]}
       ZCurve {keys [
         {interpMode=RCIM_Cubic, tangentMode=RCTM_Break, time=0, value=0.2}
         {arriveTangent=-0.1754, interpMode=RCIM_Cubic, tangentMode=RCTM_Break, time=0.6, value=1}
        ]}
      }}
     Scale Curve=<0 0 -750>
    ))
   &SolveForcesAndVelocity(Clamp Velocity=false, Limit Acceleration=false, Speed Limit=500)
   &ScaleSpriteSizeBySpeed(
    Max Scale Factor=<0.1 5>
    Min Scale Factor=<0.05 1>
    Source Velocity:Vector3f=&Multiply_Vector(A=<0.05 0.05 1>, B=$Particles.Velocity)
   )
   &ScaleColor(ScaleA=true, ScaleRGB=true)
  ]
  rendererClasses [&NiagaraSpriteRendererProperties]
  renderers [
   {rendererClass=&NiagaraSpriteRendererProperties}
   {rendererClass=&NiagaraSpriteRendererProperties, rendererIndex=1}
   {rendererClass=&NiagaraSpriteRendererProperties, rendererIndex=2}
  ]
 }}
summary {
 systemName=NS_Env_TheRaft_SewerDrips
 emitters [{emitterName=RainFromObjects, simTarget=GPUComputeSim, rendererClasses [&NiagaraSpriteRendererProperties]}]
 userVariables [
  User.Rain_AlphaMult {type=NiagaraFloat, defaultValue=1, description=""}
  User.Rain_ColorMult {type=NiagaraFloat, defaultValue=1, description=""}
  User.Rain_CylinderHeight {type=NiagaraFloat, defaultValue=1500, description=""}
  User.Rain_FallingSpeed {type=NiagaraFloat, defaultValue=1, description=""}
  User.Rain_InheritVel {type=Vector3f, description="", defaultValue {struct=Vector3f, value=<-1 -1 0>}}
  User.Rain_MaskInside_Radius {type=NiagaraFloat, defaultValue=0, description=""}
  User.Rain_NonLitColor {type=LinearColor, description="", defaultValue {struct=LinearColor, value=<rgba 1 1 1 1>}}
  User.Rain_SpawnRateScale {type=NiagaraFloat, defaultValue=2, description=""}
 ]
}
