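# Niagara system for spark effects in the Helios environment's TheRaft area, named for sparks from blue and red wires.
# It contains two emitters, SparksBlueWire and SparksRedWire, with SystemState as the only listed module dependency.
# The file lists no renderers, user variables, dynamic inputs or data interfaces, so the sparks' look and emitter behavior are not shown.
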NiagaraSystem /Game/Environments/Helios/FX/TheRaft/NS_Env_TheRaft_Sparks
dependencies {usedDataInterfaces [], usedDynamicInputs [], usedModules [&SystemState]}
emitters {SparksBlueWire {rendererClasses [], renderers []}, SparksRedWire {rendererClasses [], renderers []}}
summary {
 systemName=NS_Env_TheRaft_Sparks
 emitters [{emitterName=SparksBlueWire, rendererClasses []}, {emitterName=SparksRedWire, rendererClasses []}]
 userVariables []
}
