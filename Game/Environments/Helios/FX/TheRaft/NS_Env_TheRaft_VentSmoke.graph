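# Niagara system NS_Env_TheRaft_VentSmoke, an environment effect for TheRaft in the Helios set, named as smoke from a vent.
# It holds two emitters, SmokeVertical and SmokeCameraFacing, whose names point to an upright smoke layer paired with a camera-facing one.
# The summary lists no renderers, user variables, data interfaces or dynamic inputs, and SystemState is the only module it records.
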NiagaraSystem /Game/Environments/Helios/FX/TheRaft/NS_Env_TheRaft_VentSmoke
dependencies {usedDataInterfaces [], usedDynamicInputs [], usedModules [&SystemState]}
emitters {SmokeCameraFacing {rendererClasses [], renderers []}, SmokeVertical {rendererClasses [], renderers []}}
summary {
 systemName=NS_Env_TheRaft_VentSmoke
 emitters [{emitterName=SmokeVertical, rendererClasses []}, {emitterName=SmokeCameraFacing, rendererClasses []}]
 userVariables []
}
